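# Looping indoor spot ambience cue for broken speaker transmission interference noise, from the Utopia_Fall audio set.
# A Random node picks one of two equally weighted looping noise waves and feeds an Enveloper that applies a pitch curve and slight pitch and volume variation.
# It borrows SummerFest small-speaker attenuation, uses the indoor ambient one-shot sound class and limits concurrency through a two-instance spot setting.
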SoundCue /Hermes/Audio/Utopia_Fall/Spots/SC_Spot_Broken_Speaker_Transmission_Interference_Noise_Loop_Int
root=Enveloper_0
nodes [
 {
  id=Enveloper_0
  children [Random_0]
  properties {
   loopEnd=30
   pitchMax=1.1
   pitchMin=0.9
   volumeMin=0.9
   pitchCurve {editorCurveData {keys [{arriveTangent=-0.0033, interpMode=RCIM_Cubic, time=0}, <at 3 1.05 RCIM_Cubic>, {interpMode=RCIM_Cubic, leaveTangent=0.0033, time=5}]}}
   volumeCurve {editorCurveData {keys [<at 0 1>]}}
  }
 }
 {id=Random_0, children [WavePlayer_0,WavePlayer_2], properties {weights [1,1]}}
 {
  id=WavePlayer_0
  properties {bLooping=true, soundWaveAssetPtr=&/Spot_Broken_Speaker_Transmission_Interference_Noise_03}
 }
 {
  id=WavePlayer_2
  properties {bLooping=true, soundWaveAssetPtr=&/Spot_Broken_Speaker_Transmission_Interference_Noise_01}
 }
]
properties {
 attenuationSettings=/Game/Sounds/Artemis/Biomes/SummerFestBeachParty/BeachPartySmallSpeaker/SummerFest_Music_SmallSpeaker_Focus_Att
 bEnableSubmixSends=false
 maxDistance=1950
 soundClassObject=/Game/Sounds/FortSoundClassesAndModes/Fort_Ambient_OneShots_Indoors
 volumeMultiplier=1
 concurrencyOverrides {maxCount=2}
 concurrencySet [/AmbientAudioBase/Concurrency/AmbientConc_2Max_Spot_Concurrency]
 preEffectBusSends [{}]
}
