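# Outdoor ambient one-shot Sound Cue for distant, droning, haunting atmospheric swells in the Utopia_Fall audio set.
# A Random node picks one of seven equally weighted long swell waves.
# Two chained Envelopers shape it over time, the first bending pitch and the second gently undulating volume, before a Modulator adds slight volume variation.
# Wide one-shot attenuation and distance-based concurrency control spatial falloff and limit overlapping instances.
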SoundCue /Hermes/Audio/Utopia_Fall/Spots/SC_Spot_Distant_Droning_Haunting_Atmospheric_Swells_Long_Ext
root=Modulator_3
nodes [
 {
  id=Enveloper_0
  children [Random_0]
  properties {
   loopEnd=30
   pitchCurve {editorCurveData {keys [{arriveTangent=-0.0033, interpMode=RCIM_Cubic, time=0}, <at 7 1.05 RCIM_Cubic>, {interpMode=RCIM_Cubic, leaveTangent=0.0033, time=13}]}}
   volumeCurve {editorCurveData {keys [<at 0 1>]}}
  }
 }
 {
  id=Enveloper_3
  children [Enveloper_0]
  properties {
   loopEnd=20
   pitchCurve {editorCurveData {keys [{time=0}]}}
   volumeCurve {editorCurveData {keys [
      {arriveTangent=-0.01, interpMode=RCIM_Cubic, time=0, value=1}
      <at 4 0.8903>
      {arriveTangent=0.0121, interpMode=RCIM_Cubic, leaveTangent=0.0121, time=6, value=0.9761}
      {arriveTangent=-0.0163, interpMode=RCIM_Cubic, leaveTangent=-0.0163, time=8, value=0.9387}
      {arriveTangent=0.0059, interpMode=RCIM_Cubic, leaveTangent=0.0059, time=13, value=0.8617}
      {interpMode=RCIM_Cubic, leaveTangent=0.01, time=17, value=0.9915}
     ]}}
  }
 }
 {id=Modulator_3, children [Enveloper_3], properties {volumeMax=1, volumeMin=0.9}}
 {
  id=Random_0
  children [WavePlayer_12, WavePlayer_13, WavePlayer_14, WavePlayer_15, WavePlayer_16, WavePlayer_17, WavePlayer_18]
  properties {weights [1,1,1,1,1,1,1]}
 }
 {id=WavePlayer_12, properties {soundWaveAssetPtr=&/Spot_Distant_Droning_Haunting_Atmospheric_Swells_Long_07}}
 {id=WavePlayer_13, properties {soundWaveAssetPtr=&/Spot_Distant_Droning_Haunting_Atmospheric_Swells_Long_06}}
 {id=WavePlayer_14, properties {soundWaveAssetPtr=&/Spot_Distant_Droning_Haunting_Atmospheric_Swells_Long_05}}
 {id=WavePlayer_15, properties {soundWaveAssetPtr=&/Spot_Distant_Droning_Haunting_Atmospheric_Swells_Long_04}}
 {id=WavePlayer_16, properties {soundWaveAssetPtr=&/Spot_Distant_Droning_Haunting_Atmospheric_Swells_Long_03}}
 {id=WavePlayer_17, properties {soundWaveAssetPtr=&/Spot_Distant_Droning_Haunting_Atmospheric_Swells_Long_02}}
 {id=WavePlayer_18, properties {soundWaveAssetPtr=&/Spot_Distant_Droning_Haunting_Atmospheric_Swells_Long_01}}
]
properties {
 attenuationSettings=/AmbientAudioBase/Attenuations/Fort_TileAmbient_OneShot_Wider_Att
 bEnableSubmixSends=false
 maxDistance=11000
 soundClassObject=&Fort_Ambient_OneShots_Outdoors
 volumeMultiplier=1
 concurrencyOverrides {maxCount=2}
 concurrencySet [/AmbientAudioBase/Concurrency/AmbientConc_4Max_Spot_Distance_Concurency]
 preEffectBusSends [{}]
}
