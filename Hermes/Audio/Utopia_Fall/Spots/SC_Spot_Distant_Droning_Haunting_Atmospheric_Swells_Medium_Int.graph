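# Sound cue for distant, droning, haunting atmospheric swells, an indoor ambient one-shot spot sound in the Hermes Utopia Fall audio set.
# A Random node picks one of four equally weighted medium swell variations.
# Two chained Envelopers add slow pitch drift and gentle volume undulation, then a Modulator slightly randomizes playback volume.
# It uses wide one-shot ambient attenuation, the indoor one-shot sound class and a shared distance-based concurrency set.
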SoundCue /Hermes/Audio/Utopia_Fall/Spots/SC_Spot_Distant_Droning_Haunting_Atmospheric_Swells_Medium_Int
root=Modulator_3
nodes [
 {
  id=Enveloper_0
  children [Random_1]
  properties {
   loopEnd=30
   pitchCurve {editorCurveData {keys [
      {arriveTangent=-0.0033, interpMode=RCIM_Cubic, time=0}
      {arriveTangent=0.0083, interpMode=RCIM_Cubic, leaveTangent=0.0083, time=5}
      {arriveTangent=-0.0065, interpMode=RCIM_Cubic, leaveTangent=-0.0065, time=6, value=1.05}
      {arriveTangent=0.0047, interpMode=RCIM_Cubic, leaveTangent=0.0047, tangentMode=RCTM_User, time=10.3658, value=0.965}
     ]}}
   volumeCurve {editorCurveData {keys [<at 0 1>]}}
  }
 }
 {
  id=Enveloper_3
  children [Enveloper_0]
  properties {
   loopEnd=20
   pitchCurve {editorCurveData {keys [{time=0}]}}
   volumeCurve {editorCurveData {keys [
      {arriveTangent=-0.01, interpMode=RCIM_Cubic, time=0, value=1}
      <at 3 0.8903>
      {arriveTangent=0.0161, interpMode=RCIM_Cubic, leaveTangent=0.0161, time=4.5, value=0.9761}
      {arriveTangent=-0.0327, interpMode=RCIM_Cubic, leaveTangent=-0.0327, time=6, value=0.9387}
      {arriveTangent=0.0132, interpMode=RCIM_Cubic, leaveTangent=0.0132, time=8, value=0.8617}
      {interpMode=RCIM_Cubic, leaveTangent=0.01, time=10, value=0.9915}
     ]}}
  }
 }
 {id=Modulator_3, children [Enveloper_3], properties {volumeMax=1, volumeMin=0.9}}
 {id=Random_1, children [WavePlayer_0,WavePlayer_1,WavePlayer_2,WavePlayer_3], properties {weights [1,1,1,1]}}
 {id=WavePlayer_0, properties {soundWaveAssetPtr=&/Spot_Distant_Droning_Haunting_Atmospheric_Swells_Medium_01}}
 {id=WavePlayer_1, properties {soundWaveAssetPtr=&/Spot_Distant_Droning_Haunting_Atmospheric_Swells_Medium_02}}
 {id=WavePlayer_2, properties {soundWaveAssetPtr=&/Spot_Distant_Droning_Haunting_Atmospheric_Swells_Medium_03}}
 {id=WavePlayer_3, properties {soundWaveAssetPtr=&/Spot_Distant_Droning_Haunting_Atmospheric_Swells_Medium_04}}
]
properties {
 attenuationSettings=/AmbientAudioBase/Attenuations/Fort_TileAmbient_OneShot_Wider_Att
 bEnableSubmixSends=false
 maxDistance=11000
 soundClassObject=/Game/Sounds/FortSoundClassesAndModes/Fort_Ambient_OneShots_Indoors
 volumeMultiplier=1
 concurrencyOverrides {maxCount=2}
 concurrencySet [/AmbientAudioBase/Concurrency/AmbientConc_4Max_Spot_Distance_Concurency]
 preEffectBusSends [{}]
}
