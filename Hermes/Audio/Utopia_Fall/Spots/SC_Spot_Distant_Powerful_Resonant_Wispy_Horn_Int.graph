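# Sound cue for an indoor ambient one-shot spot in Utopia Fall, layering distant resonant growls with powerful cracks.
# Each layer randomly picks one of six waves, with the quieter crack layer starting after a random one-to-three-second delay.
# Both pass through nested, endlessly looping envelopers for pitch and volume movement, then fixed-gain modulators into the mixer.
# Indoor one-shot sound class, wide attenuation and distance-spot concurrency settings govern playback.
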SoundCue /Hermes/Audio/Utopia_Fall/Spots/SC_Spot_Distant_Powerful_Resonant_Wispy_Horn_Int
root=Mixer_0
nodes [
 {id=Delay_1, children [Random_1], properties {delayMax=3, delayMin=1}}
 {
  id=Enveloper_0
  children [Random_0]
  properties {
   bLoop=true
   bLoopIndefinitely=true
   loopEnd=30
   pitchCurve {editorCurveData {keys [{arriveTangent=-0.0033, interpMode=RCIM_Cubic, time=0}, <at 15 1.05 RCIM_Cubic>, {interpMode=RCIM_Cubic, leaveTangent=0.0033, time=30}]}}
   volumeCurve {editorCurveData {keys [<at 0 1>]}}
  }
 }
 {
  id=Enveloper_3
  children [Enveloper_0]
  properties {
   bLoop=true
   bLoopIndefinitely=true
   loopEnd=20
   pitchCurve {editorCurveData {keys [{time=0}]}}
   volumeCurve {editorCurveData {keys [
      {arriveTangent=-0.01, interpMode=RCIM_Cubic, time=0, value=1}
      {arriveTangent=-0.1156, interpMode=RCIM_Cubic, leaveTangent=-0.1156, time=2.3014, value=0.8437}
      <at 3.5902 0.5851>
      {arriveTangent=0.0199, interpMode=RCIM_Cubic, leaveTangent=0.0199, time=6.4209, value=0.8395}
      {arriveTangent=0.0118, interpMode=RCIM_Cubic, leaveTangent=0.0118, time=11.9487, value=0.7516}
      {interpMode=RCIM_Cubic, leaveTangent=0.01, time=20, value=1}
     ]}}
  }
 }
 {
  id=Enveloper_4
  children [Delay_1]
  properties {
   bLoop=true
   bLoopIndefinitely=true
   loopEnd=30
   pitchCurve {editorCurveData {keys [{arriveTangent=-0.0033, interpMode=RCIM_Cubic, time=0}, <at 15 1.05 RCIM_Cubic>, {interpMode=RCIM_Cubic, leaveTangent=0.0033, time=30}]}}
   volumeCurve {editorCurveData {keys [<at 0 1>]}}
  }
 }
 {
  id=Enveloper_5
  children [Enveloper_4]
  properties {
   bLoop=true
   bLoopIndefinitely=true
   loopEnd=20
   pitchCurve {editorCurveData {keys [{time=0}]}}
   volumeCurve {editorCurveData {keys [
      {arriveTangent=-0.01, interpMode=RCIM_Cubic, time=0, value=1}
      {arriveTangent=-0.1156, interpMode=RCIM_Cubic, leaveTangent=-0.1156, time=2.3014, value=0.8437}
      <at 3.5902 0.5851>
      {arriveTangent=0.0199, interpMode=RCIM_Cubic, leaveTangent=0.0199, time=6.4209, value=0.8395}
      {arriveTangent=0.0118, interpMode=RCIM_Cubic, leaveTangent=0.0118, time=11.9487, value=0.7516}
      {interpMode=RCIM_Cubic, leaveTangent=0.01, time=20, value=1}
     ]}}
  }
 }
 {id=Mixer_0, children [Modulator_1,Modulator_2], properties {inputVolume [0.9,0.35]}}
 {id=Modulator_1, children [Enveloper_3], properties {volumeMax=1.3, volumeMin=1.3}}
 {id=Modulator_2, children [Enveloper_5], properties {volumeMax=1.3, volumeMin=1.3}}
 {
  id=Random_0
  children [WavePlayer_17, WavePlayer_18, WavePlayer_20, WavePlayer_21, WavePlayer_22, WavePlayer_23]
  properties {weights [1,1,1,1,1,1]}
 }
 {
  id=Random_1
  children [WavePlayer_24, WavePlayer_25, WavePlayer_26, WavePlayer_27, WavePlayer_28, WavePlayer_29]
  properties {weights [1,1,1,1,1,1]}
 }
 {id=WavePlayer_17, properties {soundWaveAssetPtr=&/Spot_Distant_Resonant_Growl_06}}
 {id=WavePlayer_18, properties {soundWaveAssetPtr=&/Spot_Distant_Resonant_Growl_05}}
 {id=WavePlayer_20, properties {soundWaveAssetPtr=&/Spot_Distant_Resonant_Growl_04}}
 {id=WavePlayer_21, properties {soundWaveAssetPtr=&/Spot_Distant_Resonant_Growl_03}}
 {id=WavePlayer_22, properties {soundWaveAssetPtr=&/Spot_Distant_Resonant_Growl_02}}
 {id=WavePlayer_23, properties {soundWaveAssetPtr=&/Spot_Distant_Resonant_Growl_01}}
 {id=WavePlayer_24, properties {soundWaveAssetPtr=&/Spot_Distant_Powerful_Crack_06}}
 {id=WavePlayer_25, properties {soundWaveAssetPtr=&/Spot_Distant_Powerful_Crack_05}}
 {id=WavePlayer_26, properties {soundWaveAssetPtr=&/Spot_Distant_Powerful_Crack_04}}
 {id=WavePlayer_27, properties {soundWaveAssetPtr=&/Spot_Distant_Powerful_Crack_03}}
 {id=WavePlayer_28, properties {soundWaveAssetPtr=&/Spot_Distant_Powerful_Crack_02}}
 {id=WavePlayer_29, properties {soundWaveAssetPtr=&/Spot_Distant_Powerful_Crack_01}}
]
properties {
 attenuationSettings=/Game/Sounds/Attenuation/Fort_TileAmbient_OneShot_Wider_Att
 bEnableSubmixSends=false
 maxDistance=8000
 soundClassObject=/Game/Sounds/FortSoundClassesAndModes/Fort_Ambient_OneShots_Indoors
 volumeMultiplier=0.5
 attenuationOverrides {
  bAttenuateWithLPF=true
  bEnableListenerFocus=true
  falloffDistance=950
  lPFFrequencyAtMax=3400
  lPFRadiusMax=1000
  lPFRadiusMin=500
  nonFocusVolumeAttenuation=0.2
  attenuationShapeExtents {x=350}
 }
 concurrencyOverrides {maxCount=2, retriggerTime=2, voiceStealReleaseTime=3}
 concurrencySet [/AmbientAudioBase/Concurrency/AmbientConc_4Max_Spot_Distance_Concurency]
}
