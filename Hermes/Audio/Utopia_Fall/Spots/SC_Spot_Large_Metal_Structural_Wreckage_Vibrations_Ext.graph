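# Sound cue for an exterior ambient spot one-shot of large metal wreckage vibrations in Utopia Fall.
# A Random node picks among five equally weighted wave variations and feeds two stacked Enveloper nodes.
# The inner envelope bends pitch over time, the outer adds a gentle volume wobble, and both randomize pitch and volume slightly.
# It uses the outdoor one-shot sound class, wide one-shot attenuation and a shared distance-based spot concurrency set.
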SoundCue /Hermes/Audio/Utopia_Fall/Spots/SC_Spot_Large_Metal_Structural_Wreckage_Vibrations_Ext
root=Enveloper_3
nodes [
 {
  id=Enveloper_0
  children [Random_1]
  properties {
   loopEnd=30
   pitchMax=1.1
   pitchMin=0.9
   volumeMin=0.9
   pitchCurve {editorCurveData {keys [{arriveTangent=-0.0033, interpMode=RCIM_Cubic, time=0}, <at 3 1.05 RCIM_Cubic>, {interpMode=RCIM_Cubic, leaveTangent=0.0033, time=5}]}}
   volumeCurve {editorCurveData {keys [<at 0 1>]}}
  }
 }
 {
  id=Enveloper_3
  children [Enveloper_0]
  properties {
   loopEnd=20
   pitchMax=1.1
   pitchMin=0.9
   volumeMin=0.9
   pitchCurve {editorCurveData {keys [{time=0}]}}
   volumeCurve {editorCurveData {keys [
      {arriveTangent=-0.01, interpMode=RCIM_Cubic, time=0, value=1}
      <at 2 0.8903>
      {arriveTangent=0.0242, interpMode=RCIM_Cubic, leaveTangent=0.0242, time=3, value=0.9761}
      {arriveTangent=-0.0572, interpMode=RCIM_Cubic, leaveTangent=-0.0572, time=4, value=0.9387}
      {arriveTangent=0.0142, interpMode=RCIM_Cubic, leaveTangent=0.0142, time=5, value=0.8617}
      {interpMode=RCIM_Cubic, leaveTangent=0.01, time=7.7108, value=0.9915}
     ]}}
  }
 }
 {id=Random_1, children [WavePlayer_5,WavePlayer_6,WavePlayer_7,WavePlayer_8,WavePlayer_9], properties {weights [1,1,1,1,1]}}
 {id=WavePlayer_5, properties {soundWaveAssetPtr=&/Spot_Large_Metal_Structural_Wreckage_Vibrations_01}}
 {id=WavePlayer_6, properties {soundWaveAssetPtr=&/Spot_Large_Metal_Structural_Wreckage_Vibrations_02}}
 {id=WavePlayer_7, properties {soundWaveAssetPtr=&/Spot_Large_Metal_Structural_Wreckage_Vibrations_03}}
 {id=WavePlayer_8, properties {soundWaveAssetPtr=&/Spot_Large_Metal_Structural_Wreckage_Vibrations_04}}
 {id=WavePlayer_9, properties {soundWaveAssetPtr=&/Spot_Large_Metal_Structural_Wreckage_Vibrations_05}}
]
properties {
 attenuationSettings=/AmbientAudioBase/Attenuations/Fort_TileAmbient_OneShot_Wider_Att
 bEnableSubmixSends=false
 maxDistance=11000
 soundClassObject=&Fort_Ambient_OneShots_Outdoors
 volumeMultiplier=1
 concurrencyOverrides {maxCount=2}
 concurrencySet [/AmbientAudioBase/Concurrency/AmbientConc_4Max_Spot_Distance_Concurency]
 preEffectBusSends [{}]
}
